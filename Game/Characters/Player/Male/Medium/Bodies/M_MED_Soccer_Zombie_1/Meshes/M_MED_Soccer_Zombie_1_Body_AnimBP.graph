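# Post-processes the male medium soccer zombie body on a pose copied from its attached parent mesh, adding shorts motion and corrective bone adjustments.
# Eight Anim Dynamics chains, weighted by Shorts_Alpha, simulate the front, back, inner and outer shorts bones on both legs.
# Each update maps calf-to-thigh and right upperarm-to-clavicle yaw deltas from ref pose to alphas.
# These alphas blend Modify Bone correctives on the shorts (interpolated) and right upper-arm scale bone.
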AnimBlueprint /Game/Characters/Player/Male/Medium/Bodies/M_MED_Soccer_Zombie_1/Meshes/M_MED_Soccer_Zombie_1_Body_AnimBP
target_skeleton=&/M_MED_Soccer_Zombie_1_Body_Skeleton
graphs [
 AnimGraph {nodes [
   n1 = Get Shorts_Alpha() -> (Shorts_Alpha:double)
   n2 = "Anim Dynamics - Chain: dyn_shorts_in_l -> dyn_shorts_in_02_l"(ComponentPose=Local To Component(LocalPose=Copy Pose From Mesh() {bUseAttachedParent=true}), Alpha=@n1, ExternalForce="0,0,0") {
    angularDampingOverride=0.999
    angularSpringConstant=100
    bAngularSpring=true
    bChain=true
    bGravityOverrideInSimSpace=true
    bOverrideAngularDamping=true
    numSolverIterationsPostUpdate=4
    numSolverIterationsPreUpdate=16
    simulationSpace=RootRelative
    alphaBoolBlend {blendOption=Linear}
    boundBone {boneName=dyn_shorts_in_l}
    chainEnd {boneName=dyn_shorts_in_02_l}
    physicsBodyDefinitions [{
      sphereCollisionRadius=0
      boundBone {boneName=dyn_shorts_in_l}
      boxExtents=<25 30 12>
      constraintSetup {angularTargetAxis=AxisZ, twistAxis=AxisZ, angularLimitsMax=<15 15 10>, angularLimitsMin=<-15 -15 -10>, angularTarget=<0 0 1>}
      localJointOffset=<0 0 -3.5>
     }]
   }
   n3 = "Anim Dynamics - Chain: dyn_shorts_out_l -> dyn_shorts_out_02_l"(ComponentPose=@n2, Alpha=@n1, ExternalForce="0,0,0") {
    angularDampingOverride=0.999
    angularSpringConstant=100
    bAngularSpring=true
    bChain=true
    bGravityOverrideInSimSpace=true
    bOverrideAngularDamping=true
    numSolverIterationsPostUpdate=4
    numSolverIterationsPreUpdate=16
    simulationSpace=RootRelative
    alphaBoolBlend {blendOption=Linear}
    boundBone {boneName=dyn_shorts_out_l}
    chainEnd {boneName=dyn_shorts_out_02_l}
    physicsBodyDefinitions [{
      sphereCollisionRadius=0
      boundBone {boneName=dyn_shorts_out_l}
      boxExtents=<50 50 30>
      constraintSetup {angularTargetAxis=AxisZ, twistAxis=AxisZ, angularLimitsMax=<10 10 5>, angularLimitsMin=<-10 -10 -5>, angularTarget=<0 0 1>}
      localJointOffset=<0 0 -6.5>
     }]
   }
   n4 = "Anim Dynamics - Chain: dyn_shorts_fr_l -> dyn_shorts_fr_02_l"(ComponentPose=@n3, Alpha=@n1, ExternalForce="0,0,0") {
    angularDampingOverride=0.999
    angularSpringConstant=300
    bAngularSpring=true
    bChain=true
    bGravityOverrideInSimSpace=true
    bOverrideAngularDamping=true
    numSolverIterationsPostUpdate=4
    numSolverIterationsPreUpdate=16
    simulationSpace=RootRelative
    alphaBoolBlend {blendOption=Linear}
    boundBone {boneName=dyn_shorts_fr_l}
    chainEnd {boneName=dyn_shorts_fr_02_l}
    physicsBodyDefinitions [{
      sphereCollisionRadius=0
      boundBone {boneName=dyn_shorts_fr_l}
      boxExtents=<18 25 8>
      constraintSetup {angularTargetAxis=AxisZ, twistAxis=AxisZ, angularLimitsMax=<10 0 5>, angularLimitsMin=<-10 -20 -5>, angularTarget=<0 0 1>}
      localJointOffset=<0 0 -5>
     }]
   }
   n5 = "Anim Dynamics - Chain: dyn_shorts_bk_l -> dyn_shorts_bk_02_l"(ComponentPose=@n4, Alpha=@n1, ExternalForce="0,0,0") {
    angularDampingOverride=0.999
    angularSpringConstant=300
    bAngularSpring=true
    bChain=true
    bGravityOverrideInSimSpace=true
    bOverrideAngularDamping=true
    numSolverIterationsPostUpdate=4
    numSolverIterationsPreUpdate=16
    simulationSpace=RootRelative
    alphaBoolBlend {blendOption=Linear}
    boundBone {boneName=dyn_shorts_bk_l}
    chainEnd {boneName=dyn_shorts_bk_02_l}
    physicsBodyDefinitions [{
      sphereCollisionRadius=0
      boundBone {boneName=dyn_shorts_bk_l}
      boxExtents=<40 38 15>
      constraintSetup {angularTargetAxis=AxisZ, twistAxis=AxisZ, angularLimitsMax=<15 0 8>, angularLimitsMin=<-15 -20 -8>, angularTarget=<0 0 1>}
      localJointOffset=<0 0 -3.5>
     }]
   }
   n6 = "Anim Dynamics - Chain: dyn_shorts_in_r -> dyn_shorts_in_02_r"(ComponentPose=@n5, Alpha=@n1, ExternalForce="0,0,0") {
    angularDampingOverride=0.999
    angularSpringConstant=100
    bAngularSpring=true
    bChain=true
    bGravityOverrideInSimSpace=true
    bOverrideAngularDamping=true
    numSolverIterationsPostUpdate=4
    numSolverIterationsPreUpdate=16
    simulationSpace=RootRelative
    alphaBoolBlend {blendOption=Linear}
    boundBone {boneName=dyn_shorts_in_r}
    chainEnd {boneName=dyn_shorts_in_02_r}
    physicsBodyDefinitions [{
      sphereCollisionRadius=0
      boundBone {boneName=dyn_shorts_in_r}
      boxExtents=<25 29 12>
      constraintSetup {angularTargetAxis=AxisZ, twistAxis=AxisZ, angularLimitsMax=<15 10 15>, angularLimitsMin=<-15 -20 -15>, angularTarget=<0 0 1>}
      localJointOffset=<0 0 3.5>
     }]
   }
   n7 = "Anim Dynamics - Chain: dyn_shorts_out_r -> dyn_shorts_out_02_r"(ComponentPose=@n6, Alpha=@n1, ExternalForce="0,0,0") {
    angularDampingOverride=0.999
    angularSpringConstant=100
    bAngularSpring=true
    bChain=true
    bGravityOverrideInSimSpace=true
    bOverrideAngularDamping=true
    numSolverIterationsPostUpdate=4
    numSolverIterationsPreUpdate=16
    simulationSpace=RootRelative
    alphaBoolBlend {blendOption=Linear}
    boundBone {boneName=dyn_shorts_out_r}
    chainEnd {boneName=dyn_shorts_out_02_r}
    physicsBodyDefinitions [{
      sphereCollisionRadius=0
      boundBone {boneName=dyn_shorts_out_r}
      boxExtents=<50 50 30>
      constraintSetup {angularTargetAxis=AxisZ, twistAxis=AxisZ, angularLimitsMax=<10 10 5>, angularLimitsMin=<-10 -10 -5>, angularTarget=<0 0 1>}
      localJointOffset=<1 0 6.5>
     }]
   }
   n8 = "Anim Dynamics - Chain: dyn_shorts_fr_r -> dyn_shorts_fr_02_r"(ComponentPose=@n7, Alpha=@n1, ExternalForce="0,0,0") {
    angularDampingOverride=0.999
    angularSpringConstant=300
    bAngularSpring=true
    bChain=true
    bGravityOverrideInSimSpace=true
    bOverrideAngularDamping=true
    numSolverIterationsPostUpdate=4
    numSolverIterationsPreUpdate=16
    simulationSpace=RootRelative
    alphaBoolBlend {blendOption=Linear}
    boundBone {boneName=dyn_shorts_fr_r}
    chainEnd {boneName=dyn_shorts_fr_02_r}
    physicsBodyDefinitions [{
      sphereCollisionRadius=0
      boundBone {boneName=dyn_shorts_fr_r}
      boxExtents=<26 22 8>
      constraintSetup {angularTargetAxis=AxisZ, twistAxis=AxisZ, angularLimitsMax=<10 0 5>, angularLimitsMin=<-10 -20 -5>, angularTarget=<0 0 1>}
      localJointOffset=<0.5 0 5>
     }]
   }
   n9 = "Anim Dynamics - Chain: dyn_shorts_bk_r -> dyn_shorts_bk_02_r"(ComponentPose=@n8, Alpha=@n1, ExternalForce="0,0,0") {
    angularDampingOverride=0.999
    angularSpringConstant=300
    bAngularSpring=true
    bChain=true
    bGravityOverrideInSimSpace=true
    bOverrideAngularDamping=true
    numSolverIterationsPostUpdate=4
    numSolverIterationsPreUpdate=16
    simulationSpace=RootRelative
    alphaBoolBlend {blendOption=Linear}
    boundBone {boneName=dyn_shorts_bk_r}
    chainEnd {boneName=dyn_shorts_bk_02_r}
    physicsBodyDefinitions [{
      sphereCollisionRadius=0
      boundBone {boneName=dyn_shorts_bk_r}
      boxExtents=<40 36 15>
      constraintSetup {angularTargetAxis=AxisZ, twistAxis=AxisZ, angularLimitsMax=<20 5 15>, angularLimitsMin=<-20 -20 -15>, angularTarget=<0 0 1>}
      localJointOffset=<0.5 0 3.5>
     }]
   }
   n10 = Get calf_rot_back_l() -> (calf_rot_back_l:double)
   n11 = "Transform (Modify) Bone - Bone: dyn_shorts_main_l"(ComponentPose=@n9, Alpha=@n10, Translation="0,0,0", Rotation="0,0,10", Scale="1,1,1") {
    rotationMode=BMM_Additive
    rotationSpace=BCS_BoneSpace
    scaleMode=BMM_Additive
    scaleSpace=BCS_BoneSpace
    translationMode=BMM_Additive
    translationSpace=BCS_BoneSpace
    alphaBoolBlend {blendOption=Linear}
    alphaScaleBiasClamp {bInterpResult=true, interpSpeedDecreasing=25, interpSpeedIncreasing=25}
    boneToModify {boneName=dyn_shorts_main_l}
    rotation {roll=10}
   }
   n12 = "Transform (Modify) Bone - Bone: dyn_shorts_bk_l"(ComponentPose=@n11, Alpha=@n10, Translation="-0.247,0.000,5", Rotation="24.730,-3.101,-11.925", Scale="0.9,0.9,0.9") {
    rotationMode=BMM_Additive
    rotationSpace=BCS_BoneSpace
    scaleMode=BMM_Additive
    scaleSpace=BCS_BoneSpace
    translationMode=BMM_Additive
    translationSpace=BCS_BoneSpace
    alphaBoolBlend {blendOption=Linear}
    alphaScaleBiasClamp {bInterpResult=true, interpSpeedDecreasing=25, interpSpeedIncreasing=25}
    boneToModify {boneName=dyn_shorts_bk_l}
    rotation=<rot 24.73 -3.101 -11.925>
    scale=<0.9 0.9 0.9>
    translation {x=-0.247, z=5}
   }
   n13 = "Transform (Modify) Bone - Bone: dyn_shorts_out_l"(ComponentPose=@n12, Alpha=@n10, Translation="-0.000,0.022,2", Rotation="8,0.000,0.000", Scale="1,1,1") {
    rotationMode=BMM_Additive
    rotationSpace=BCS_BoneSpace
    scaleMode=BMM_Additive
    scaleSpace=BCS_BoneSpace
    translationMode=BMM_Additive
    translationSpace=BCS_BoneSpace
    alphaBoolBlend {blendOption=Linear}
    alphaScaleBiasClamp {bInterpResult=true, interpSpeedDecreasing=25, interpSpeedIncreasing=25}
    boneToModify {boneName=dyn_shorts_out_l}
    rotation {pitch=8}
    translation {y=0.022, z=2}
   }
   n14 = "Transform (Modify) Bone - Bone: dyn_shorts_in_l"(ComponentPose=@n13, Alpha=@n10, Translation="0.000,-0.033,4", Rotation="-12,0.000,0.000", Scale="1,1,1") {
    rotationMode=BMM_Additive
    rotationSpace=BCS_BoneSpace
    scaleMode=BMM_Additive
    scaleSpace=BCS_BoneSpace
    translationMode=BMM_Additive
    translationSpace=BCS_BoneSpace
    alphaBoolBlend {blendOption=Linear}
    alphaScaleBiasClamp {bInterpResult=true, interpSpeedDecreasing=25, interpSpeedIncreasing=25}
    boneToModify {boneName=dyn_shorts_in_l}
    rotation {pitch=-12}
    translation {y=-0.033, z=4}
   }
   n15 = "Transform (Modify) Bone - Bone: dyn_shorts_fr_l"(ComponentPose=@n14, Alpha=@n10, Translation="2.3,1,0", Rotation="-5,0,0", Scale="1,1,1") {
    rotationMode=BMM_Additive
    rotationSpace=BCS_BoneSpace
    scaleMode=BMM_Additive
    scaleSpace=BCS_BoneSpace
    translationMode=BMM_Additive
    translationSpace=BCS_BoneSpace
    alphaBoolBlend {blendOption=Linear}
    alphaScaleBiasClamp {bInterpResult=true, interpSpeedDecreasing=25, interpSpeedIncreasing=25}
    boneToModify {boneName=dyn_shorts_fr_l}
    rotation {pitch=-5}
    translation=<2.3 1>
   }
   n16 = Get calf_rot_back_r() -> (calf_rot_back_r:double)
   n17 = "Transform (Modify) Bone - Bone: dyn_shorts_main_r"(ComponentPose=@n15, Alpha=@n16, Translation="0,0,0", Rotation="0,0,10", Scale="1,1,1") {
    rotationMode=BMM_Additive
    rotationSpace=BCS_BoneSpace
    scaleMode=BMM_Additive
    scaleSpace=BCS_BoneSpace
    translationMode=BMM_Additive
    translationSpace=BCS_BoneSpace
    alphaBoolBlend {blendOption=Linear}
    alphaScaleBiasClamp {bInterpResult=true, interpSpeedDecreasing=25, interpSpeedIncreasing=25}
    boneToModify {boneName=dyn_shorts_main_r}
    rotation {roll=10}
   }
   n18 = "Transform (Modify) Bone - Bone: dyn_shorts_bk_r"(ComponentPose=@n17, Alpha=@n16, Translation="-0.247,0.000,-5", Rotation="24.730,-3.101,-11.925", Scale="0.9,0.9,0.9") {
    rotationMode=BMM_Additive
    rotationSpace=BCS_BoneSpace
    scaleMode=BMM_Additive
    scaleSpace=BCS_BoneSpace
    translationMode=BMM_Additive
    translationSpace=BCS_BoneSpace
    alphaBoolBlend {blendOption=Linear}
    alphaScaleBiasClamp {bInterpResult=true, interpSpeedDecreasing=25, interpSpeedIncreasing=25}
    boneToModify {boneName=dyn_shorts_bk_r}
    rotation=<rot 24.73 -3.101 -11.925>
    scale=<0.9 0.9 0.9>
    translation {x=-0.247, z=-5}
   }
   n19 = "Transform (Modify) Bone - Bone: dyn_shorts_out_r"(ComponentPose=@n18, Alpha=@n16, Translation="-0.000,0.022,-2", Rotation="8,0.000,0.000", Scale="1,1,1") {
    rotationMode=BMM_Additive
    rotationSpace=BCS_BoneSpace
    scaleMode=BMM_Additive
    scaleSpace=BCS_BoneSpace
    translationMode=BMM_Additive
    translationSpace=BCS_BoneSpace
    alphaBoolBlend {blendOption=Linear}
    alphaScaleBiasClamp {bInterpResult=true, interpSpeedDecreasing=25, interpSpeedIncreasing=25}
    boneToModify {boneName=dyn_shorts_out_r}
    rotation {pitch=8}
    translation {y=0.022, z=-2}
   }
   n20 = "Transform (Modify) Bone - Bone: dyn_shorts_in_r"(ComponentPose=@n19, Alpha=@n16, Translation="0.000,-0.033,-4", Rotation="0,0,0", Scale="1,1,1") {
    rotationMode=BMM_Additive
    rotationSpace=BCS_BoneSpace
    scaleMode=BMM_Additive
    scaleSpace=BCS_BoneSpace
    translationMode=BMM_Additive
    translationSpace=BCS_BoneSpace
    alphaBoolBlend {blendOption=Linear}
    alphaScaleBiasClamp {bInterpResult=true, interpSpeedDecreasing=25, interpSpeedIncreasing=25}
    boneToModify {boneName=dyn_shorts_in_r}
    translation {y=-0.033, z=-4}
   }
   n21 = "Transform (Modify) Bone - Bone: dyn_shorts_fr_r"(ComponentPose=@n20, Alpha=@n16, Translation="-1,0.4,0.109", Rotation="0,0,0", Scale="1,1,1") {
    rotationMode=BMM_Additive
    rotationSpace=BCS_BoneSpace
    scaleMode=BMM_Additive
    scaleSpace=BCS_BoneSpace
    translationMode=BMM_Additive
    translationSpace=BCS_BoneSpace
    alphaBoolBlend {blendOption=Linear}
    alphaScaleBiasClamp {bInterpResult=true, interpSpeedDecreasing=25, interpSpeedIncreasing=25}
    boneToModify {boneName=dyn_shorts_fr_r}
    translation=<-1 0.4 0.109>
   }
   n22 = "Transform (Modify) Bone - Bone: dyn_upperarmScale_r"(
    ComponentPose=@n21
    Alpha=Get armSwingDelta_r() -> (armSwingDelta_r:double)
    Translation="-12,-20,20"
    Rotation="-35,0,0"
    Scale="1.5,0.15,1.25"
   ) {
    rotationMode=BMM_Additive
    rotationSpace=BCS_BoneSpace
    scaleMode=BMM_Replace
    scaleSpace=BCS_BoneSpace
    translationMode=BMM_Additive
    translationSpace=BCS_BoneSpace
    alphaBoolBlend {blendOption=Linear}
    boneToModify {boneName=dyn_upperarmScale_r}
    rotation {pitch=-35}
    scale=<1.5 0.15 1.25>
    translation=<-12 -20 20>
   }
   Output Pose(Result=Component To Local(ComponentPose=@n22)) {layerGroup=None}
  ]}
 EventGraph {nodes [
   Event BlueprintUpdateAnimation()
   > n1 = Sequence!(execute) -> (then_0, then_1, then_2:Exec, then_3:Exec)
   n2 = GetOwnerSkeletalMeshComponent()
   n3 = BreakRotator(InRot=BreakTransform(InTransform=GetDeltaTransformFromRefPose(self=@n2, BoneName=upperarm_r, BaseName=clavicle_r)).Rotation)
   Set armSwingDelta_r(execute=@n1.then_2, armSwingDelta_r:double=MapRangeClamped(Value="Absolute (Float)"(A=@n3.Yaw), InRangeA=75, InRangeB=125)) -> (Output_Get:double)
   n4 = BreakTransform(InTransform=GetDeltaTransformFromRefPose(self=@n2, BoneName=calf_l, BaseName=thigh_l))
   Set calf_rot_back_l(execute=@n1.then_1, calf_rot_back_l:double=MapRangeClamped(Value=BreakRotator(InRot=@n4.Rotation).Yaw, InRangeA=15, InRangeB=115)) -> (Output_Get:double)
   n5 = BreakTransform(InTransform=GetDeltaTransformFromRefPose(self=GetOwnerSkeletalMeshComponent(), BoneName=calf_r, BaseName=thigh_r))
   Set calf_rot_back_r(execute=@n1.then_0, calf_rot_back_r:double=MapRangeClamped(Value=BreakRotator(InRot=@n5.Rotation).Yaw, InRangeA=15, InRangeB=115)) -> (Output_Get:double)
   TryGetPawnOwner()
  ]}
]
variables [
 armSwingDelta_r {type=number}
 calf_rot_back_l {type=number}
 calf_rot_back_r {type=number}
 Shorts_Alpha {type=number, default=0.6}
]
